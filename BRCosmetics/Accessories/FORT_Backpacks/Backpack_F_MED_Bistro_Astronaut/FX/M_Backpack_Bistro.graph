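# Masked, two-sided lit material for Niagara mesh particles on the Bistro backpack, giving an animated pink, liquid-like surface.
# Panning distortion normals blend dark and light colors for particle-tinted emissive, while a second, UV-warped pan supplies the exaggerated surface normal.
# The cutout multiplies V-axis sphere masks, one eroded by that distortion, with thresholded panning water foam.
# Dynamic parameters drive both mask radii and the foam threshold, letting particles animate reveal and breakup.
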Material /BRCosmetics/Accessories/FORT_Backpacks/Backpack_F_MED_Bistro_Astronaut/FX/M_Backpack_Bistro
graph {
 properties {bUsedWithNiagaraMeshParticles=true, blendMode=BLEND_Masked, pixelDepthOffsetMode=PDOM_Legacy, twoSided=true, usedShadingModels=Default Lit}
 nodes [
  n1 = VectorParameter() {parameterName=DarkColor, defaultValue=<rgba 0.026 0.003 0.0088 1>}
  n2 = VectorParameter() {parameterName=LightColor, defaultValue=<rgba 0.6823 0.0824 0.2195 1>}
  n3 = TextureCoordinate() {uTiling=0.2}
  n4 = MaterialFunctionCall(f2=TextureSample(UVs=Panner(Coordinate=@n3) {speedX=0.3}) {samplerType=SAMPLERTYPE_Normal, texture=/Game/Global/Textures/Default/Tileable_Patterns/T_Tilling_Distortion_Map_01_n}.RGB) {materialFunction=&AddComponents}
  n5 = Abs(None=Clamp(None=Power(Base=@n4.f2) {constExponent=3}))
  MaterialFunctionCall(Emissive=Multiply(A=LinearInterpolate(A=@n1.RGB, B=@n2.RGB, Alpha=@n5), B=ParticleColor().RGB)) {materialFunction=&MF_FX_EmissiveAdjustments}
  n6 = TextureCoordinate() {uTiling=0.2, vTiling=0.77}
  n7 = Multiply(A=TextureSample(UVs=TextureCoordinate() {uTiling=0.4, vTiling=0.3}) {samplerType=SAMPLERTYPE_Normal, texture=/Game/Packages/Fortress_Sky/Materials/2Normal}.R) {constB=0.04}
  n8 = TextureSample(UVs=Panner(Coordinate=Add(A=@n6, B=@n7)) {speedX=0.42}) {samplerType=SAMPLERTYPE_Normal, texture=&T_Tilling_Distortion_Map_01_n}
  MaterialFunctionCall(Normal=@n8.RGB, Flatness=ScalarParameter() {defaultValue=-4, parameterName=Flatness}) {materialFunction=&FlattenNormal}
  n9 = ComponentMask(None=TextureCoordinate()) {g=true}
  n10 = Constant()
  n11 = DynamicParameter() {defaultValue=<rgba 0 1 0 0>}
  n12 = SphereMask(A=@n9, B=@n10, Radius=@n11.Param1) {hardnessPercent=50}
  n13 = Constant() {r=1}
  n14 = Panner(Coordinate=TextureCoordinate() {uTiling=0.3, vTiling=0.35}) {speedX=0.1, speedY=0.2}
  n15 = Constant() {r=1}
  Multiply(
   A=Multiply(A=If(A=@n12, B=@n8.R, "A > B"=@n13, "A == B"=@n13, "A < B"=@n10), B=SphereMask(A=@n9, B=@n10, Radius=@n11.Param2) {hardnessPercent=0})
   B=If(
    A=TextureSample(UVs=@n14) {samplerType=SAMPLERTYPE_LinearColor, texture=/Game/Athena/Apollo/Water/Textures/02_Rounded/03_WaterFoam_Veroni_base_H}.R
    B=@n11.Param3
    "A > B"=@n15
    "A == B"=@n15
    "A < B"=Constant()
   )
  )
  ScalarParameter() {defaultValue=0.22, parameterName=Roughness}
  VectorParameter() {parameterName=BaseColor, defaultValue=<rgba 0.0781 0.0003 0.0208 1>}
 ]
}
