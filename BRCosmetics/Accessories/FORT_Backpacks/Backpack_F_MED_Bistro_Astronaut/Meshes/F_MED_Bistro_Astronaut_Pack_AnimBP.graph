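# Animation Blueprint for the Bistro Astronaut backpack cosmetic, playing idles and a switch animation into the Monster style.
# A state machine feeds DefaultSlot, starting in Idle, where a weighted random player favors the base idle over a break.
# The anim input event aie.BistroBackpackOutro sets the Transform bool, triggering the one-way move to Transform To Monster.
# Two test states auto-loop between the idle clips but are unreachable from entry.
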AnimBlueprint /BRCosmetics/Accessories/FORT_Backpacks/Backpack_F_MED_Bistro_Astronaut/Meshes/F_MED_Bistro_Astronaut_Pack_AnimBP
target_skeleton=&/F_MED_Bistro_Astronaut_Pack_Skeleton
graphs [
 AnimGraph {nodes [
   AnimGraphNode_StateMachine_0 = Idle State Machine() {is=title AnimGraphNode_StateMachine}
   Output Pose(Result=Slot 'DefaultSlot'(Source=@AnimGraphNode_StateMachine_0)) {layerGroup=None}
  ]}
 Idle State Machine {
  entry=Idle
  path=AnimGraph.AnimGraphNode_StateMachine_0.Idle State Machine
  ids=id
  nodes [
   AnimStateNode_0 = Idle()
   AnimStateNode_1 = TestLoop() {is=title AnimStateNode}
   AnimStateNode_2 = Test Loop() {is=title AnimStateNode}
   AnimStateNode_3 = Transform To Monster() {is=title AnimStateNode}
   AnimStateTransitionNode_0 = AnimStateTransitionNode() {bAutomaticRuleBasedOnSequencePlayerInState=true}
   AnimStateTransitionNode_1 = AnimStateTransitionNode() {bAutomaticRuleBasedOnSequencePlayerInState=true}
   AnimStateTransitionNode_2 = AnimStateTransitionNode()
  ]
  links [
   {
    dst=Transform To Monster
    rule=AnimGraph.AnimGraphNode_StateMachine_0.Idle State Machine.AnimStateTransitionNode_2.Transition
    src=Idle
    via=AnimStateTransitionNode_2
   }
   {
    dst=TestLoop
    rule=AnimGraph.AnimGraphNode_StateMachine_0.Idle State Machine.AnimStateTransitionNode_1.Transition
    src=Test Loop
    via=AnimStateTransitionNode_1
   }
   {
    dst=Test Loop
    rule=AnimGraph.AnimGraphNode_StateMachine_0.Idle State Machine.AnimStateTransitionNode_0.Transition
    src=TestLoop
    via=AnimStateTransitionNode_0
   }
  ]
 }
 Idle {
  path=AnimGraph.AnimGraphNode_StateMachine_0.Idle State Machine.AnimStateNode_0.Idle
  nodes [
   n1 = Random Sequence Player!() -> (Pose:Pose Link Structure) {
    is=title AnimGraphNode_RandomPlayer
    entries [
     {
      chanceToPlay=0.8
      maxLoopCount=0
      maxPlayRate=1
      minLoopCount=0
      minPlayRate=1
      sequence=/BRCosmetics/Accessories/FORT_Backpacks/Backpack_F_MED_Bistro_Astronaut/Animation/Bistro_Astronaut_Female_Idle_BACKPACK
      blendIn {blendOption=Linear, blendTime=0.2, customCurve=null}
     }
     {
      chanceToPlay=0.2
      maxLoopCount=0
      maxPlayRate=1
      minLoopCount=0
      minPlayRate=1
      sequence=/BRCosmetics/Accessories/FORT_Backpacks/Backpack_F_MED_Bistro_Astronaut/Animation/Bistro_Astronaut_Female_Idle_Break_BACKPACK
      blendIn {blendOption=Linear, blendTime=0.2, customCurve=null}
     }
    ]
   }
   Output Animation Pose(Result=@n1)
  ]
 }
 TestLoop {
  path=AnimGraph.AnimGraphNode_StateMachine_0.Idle State Machine.AnimStateNode_1.TestLoop
  nodes [Output Animation Pose(Result=Sequence Player 'Bistro_Astronaut_Female_Idle_BACKPACK'() {sequence=&Bistro_Astronaut_Female_Idle_BACKPACK})]
 }
 Test Loop {
  path=AnimGraph.AnimGraphNode_StateMachine_0.Idle State Machine.AnimStateNode_2.Test Loop
  nodes [Output Animation Pose(Result=Sequence Player 'Bistro_Astronaut_Female_Idle_Break_BACKPACK'() {sequence=&Bistro_Astronaut_Female_Idle_Break_BACKPACK})]
 }
 Transform To Monster {
  path=AnimGraph.AnimGraphNode_StateMachine_0.Idle State Machine.AnimStateNode_3.Transform To Monster
  nodes [Output Animation Pose(Result=Sequence Player 'Emote_BistroStyle_Switch_Astronaut_Backpack'() {sequence=/BRCosmetics/Accessories/FORT_Backpacks/Backpack_F_MED_Bistro_Monster/Animation/Emote_BistroStyle_Switch_Astronaut_Backpack})]
 }
 Transition {path=AnimGraph.AnimGraphNode_StateMachine_0.Idle State Machine.AnimStateTransitionNode_0.Transition, nodes [Result()]}
 Transition {path=AnimGraph.AnimGraphNode_StateMachine_0.Idle State Machine.AnimStateTransitionNode_1.Transition, nodes [Result()]}
 Transition {
  path=AnimGraph.AnimGraphNode_StateMachine_0.Idle State Machine.AnimStateTransitionNode_2.Transition
  nodes [Result(bCanEnterTransition=Get Transform() -> (Transform:bool))]
 }
 EventGraph {nodes [
   n1 = Event BlueprintProcessAnimInputEvent()
   > Branch(Condition=IsNamed(self=@n1.AnimInputEvent, InEventName=aie.BistroBackpackOutro))
   > Set Transform(Transform:bool=true) -> (Output_Get:bool)
   TryGetPawnOwner()
   Event BlueprintUpdateAnimation()
  ]}
]
variables [Transform {type=boolean}]
